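# Physics asset for the collar mesh of the M_MED_DryEraseCod male medium player body cosmetic, providing simplified upper-torso collision shapes.
# Three capsule bodies cover the upper spine (spine_05) and both upper arms, with the arm capsules roughly mirrored.
# Each upper arm is constrained to spine_05 with free swing and twist, so the joints impose no angular limits.
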PhysicsAsset /BRCosmetics/Characters/Player/Male/Medium/Bodies/M_MED_DryEraseCod/Meshes/M_MED_DryEraseCod_Collar_Physics
bodies {
 spine_05 [{
   shapeType=Capsule
   length=19.1185
   radius=10.5604
   shapeName=spine_05_capsule
   center=<1.3236 -0.003 0.1645>
   rotation=<rot 1.9686 -90 -1.2338>
  }]
 upperarm_l [{
   shapeType=Capsule
   length=24.0467
   radius=6
   shapeName=upperarm_l_capsule
   center=<8.6035 -1.2278 -0.752>
   rotation=<rot 83.8417 -0.0003 -0.0003>
  }]
 upperarm_r [{
   shapeType=Capsule
   length=24.0467
   radius=6
   shapeName=upperarm_r_capsule
   center=<-8.6035 0.949 0.752>
   rotation=<rot 83.8417 -0.0003 179.9997>
  }]
}
constraints [
 {
  bone1Name=upperarm_l
  bone2Name=spine_05
  swing1LimitDegrees=45
  swing1Motion=Free
  swing2LimitDegrees=45
  swing2Motion=Free
  twistLimitDegrees=45
  twistMotion=Free
 }
 {
  bone1Name=upperarm_r
  bone2Name=spine_05
  swing1LimitDegrees=45
  swing1Motion=Free
  swing2LimitDegrees=45
  swing2Motion=Free
  twistLimitDegrees=45
  twistMotion=Free
 }
]
