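# Physics asset for the DryEraseCod male medium player body, set up for secondary motion of the shirt's front, side and back flaps.
# Each flap has a zero-size master body locked in rotation to spine_02, with a box or capsule child swinging from it under angular limits.
# Capsules on pelvis, spine_02 and thighs provide body volumes around the cloth.
# Front flaps are asymmetric, the right having a thinner box and looser limits.
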PhysicsAsset /BRCosmetics/Characters/Player/Male/Medium/Bodies/M_MED_DryEraseCod/Meshes/M_MED_DryEraseCod_FR_Physics
bodies {
 dyn_L_backShirt_1 [{shapeType=Capsule, length=7, radius=3, center=<6.0965 -2.9396 0>}]
 dyn_L_backShirt_MSTR [{shapeType=Capsule, length=0, radius=0, shapeName=dyn_L_backShirt_MSTR_capsule}]
 dyn_L_frontShirt_1 [{
   shapeType=Box
   extentX=9
   extentY=2.397
   extentZ=14
   shapeName=dyn_L_frontShirt_1_box
   center=<5.5411 -1.0292 -0.4919>
   rotation=<rot 0.0002 -1.8502 3.8042>
  }]
 dyn_L_frontShirt_MSTR [{shapeType=Capsule, length=0, radius=0, shapeName=dyn_L_frontShirt_MSTR_capsule}]
 dyn_L_sideShirt_1 [{shapeType=Capsule, length=10, radius=3, center=<6 -1.2684 0>}]
 dyn_L_sideShirt_MSTR [{shapeType=Capsule, length=0, radius=0, shapeName=dyn_L_sideShirt_MSTR_capsule}]
 dyn_R_backShirt_1 [{shapeType=Capsule, length=7, radius=3, center=<6.0965 -2.9396 0>}]
 dyn_R_backShirt_MSTR [{shapeType=Capsule, length=0, radius=0, shapeName=dyn_R_backShirt_MSTR_capsule}]
 dyn_R_frontShirt_1 [{shapeType=Box, extentX=9, extentY=1.4537, extentZ=14, shapeName=dyn_L_frontShirt_1_box, center=<5.5468 0 -0.5035>}]
 dyn_R_frontShirt_MSTR [{shapeType=Capsule, length=0, radius=0, shapeName=dyn_R_frontShirt_MSTR_capsule}]
 dyn_R_sideShirt_1 [{shapeType=Capsule, length=7, radius=3, center=<6 -1.2684 0>}]
 dyn_R_sideShirt_MSTR [{shapeType=Capsule, length=0, radius=0, shapeName=dyn_R_sideShirt_MSTR_capsule}]
 pelvis [{shapeType=Capsule, length=10.1875, radius=8.8957, center=<6.1102 -0.7932 0>}]
 spine_02 [{shapeType=Capsule, length=4.5, radius=10.3438, shapeName=spine_02_capsule, center=<0 0.8226 0>, rotation=<rot 0 -90 0>}]
 thigh_l [{shapeType=Capsule, length=33.9218, radius=8.6599, center=<-13.7151 -0.7814 0.8202>, rotation=<rot -7.0155 89.2064 -88.8138>}]
 thigh_r [{shapeType=Capsule, length=36.1149, radius=8.6599, center=<13.7151 0.7814 -0.8202>, rotation=<rot -7.0685 92.2717 90.8104>}]
}
constraints [
 {bone1Name=dyn_L_backShirt_MSTR, bone2Name=spine_02, swing1LimitDegrees=0, swing2LimitDegrees=0, twistLimitDegrees=0}
 {bone1Name=dyn_L_backShirt_1, bone2Name=dyn_L_backShirt_MSTR, swing1LimitDegrees=45, swing2LimitDegrees=45, twistLimitDegrees=0}
 {bone1Name=dyn_L_frontShirt_MSTR, bone2Name=spine_02, swing1LimitDegrees=0, swing2LimitDegrees=0, twistLimitDegrees=0}
 {bone1Name=dyn_L_frontShirt_1, bone2Name=dyn_L_frontShirt_MSTR, swing1LimitDegrees=50, swing2LimitDegrees=15, twistLimitDegrees=30}
 {bone1Name=dyn_L_sideShirt_MSTR, bone2Name=spine_02, swing1LimitDegrees=0, swing2LimitDegrees=0, twistLimitDegrees=0}
 {bone1Name=dyn_L_sideShirt_1, bone2Name=dyn_L_sideShirt_MSTR, swing1LimitDegrees=25, swing2LimitDegrees=25, twistLimitDegrees=25}
 {bone1Name=dyn_R_backShirt_MSTR, bone2Name=spine_02, swing1LimitDegrees=0, swing2LimitDegrees=0, twistLimitDegrees=0}
 {bone1Name=dyn_R_backShirt_1, bone2Name=dyn_R_backShirt_MSTR, swing1LimitDegrees=45, swing2LimitDegrees=45, twistLimitDegrees=0}
 {bone1Name=dyn_R_frontShirt_MSTR, bone2Name=spine_02, swing1LimitDegrees=0, swing2LimitDegrees=0, twistLimitDegrees=0}
 {bone1Name=dyn_R_frontShirt_1, bone2Name=dyn_R_frontShirt_MSTR, swing1LimitDegrees=80, swing2LimitDegrees=35, twistLimitDegrees=45}
 {bone1Name=dyn_R_sideShirt_MSTR, bone2Name=spine_02, swing1LimitDegrees=0, swing2LimitDegrees=0, twistLimitDegrees=0}
 {bone1Name=dyn_R_sideShirt_1, bone2Name=dyn_R_sideShirt_MSTR, swing1LimitDegrees=25, swing2LimitDegrees=25, twistLimitDegrees=25}
]
