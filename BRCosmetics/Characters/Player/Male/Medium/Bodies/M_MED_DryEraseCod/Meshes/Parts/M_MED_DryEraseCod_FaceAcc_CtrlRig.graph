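# Control Rig for the DryEraseCod male face accessory that applies pose-driven corrective transforms to the dynamic hat bones.
# Spherical pose readers weight Modify Transforms nodes: two track head against spine_05, the second applying only when Emote is true.
# A third tracks clavicle_r against head and overrides local hat transforms.
# A fourth reader, a spine-to-neck angle multiplier and the Crouching variable are left unconnected.
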ControlRigBlueprint /BRCosmetics/Characters/Player/Male/Medium/Bodies/M_MED_DryEraseCod/Meshes/Parts/M_MED_DryEraseCod_FaceAcc_CtrlRig
graphs [
 RigVMFunctionLibrary {links [], nodes []}
 RigVMFunction_Sequence_1_ContainedGraph {nodes [
   n1 = RigVMFunctionEntryNode()
   RigVMFunction_Sequence = &RigVMFunction_Sequence(ExecuteContext=@n1.ExecuteContext)
   n2 = &RigVMFunction_Sequence(ExecuteContext=@RigVMFunction_Sequence.B)
   RigVMFunctionReturnNode(A=@RigVMFunction_Sequence.A, B=@n2.A, C=@n2.B)
  ]}
 RigVMModel {nodes [
   RigUnit_BeginExecution = &RigUnit_BeginExecution()
   RigVMFunction_Sequence = RigVMAggregateNode(ExecuteContext=@RigUnit_BeginExecution.ExecutePin)
   n1 = &RigUnit_SphericalPoseReader(
    ActiveRegionScaleFactors=(PositiveWidth=1,NegativeWidth=1,PositiveHeight=1,NegativeHeight=1)
    ActiveRegionSize=0.1
    Debug=(bDrawDebug=true,bDraw2D=False,bDrawLocalAxes=False,DebugScale=25,DebugSegments=20,DebugThickness=0.25)
    DriverAxis=(X=1,Y=0,Z=0)
    DriverItem=(Type=Bone,Name="head")
    FalloffRegionScaleFactors=(PositiveWidth=1,NegativeWidth=1,PositiveHeight=1,NegativeHeight=1)
    FalloffSize=0.35
    FlipHeightScaling=False
    FlipWidthScaling=False
    OptionalParentItem=(Type=Bone,Name="spine_05")
    RotationOffset=(X=-30,Y=-50,Z=0)
    ExecutePin=@RigVMFunction_Sequence.A
   )
   > &RigUnit_ModifyTransforms(
    ItemToModify="((Item=(Type=Bone,Name=\"dyn_hat_rear_01\"),Transform=(Rotation=(X=-0.0868,Y=-0.0868,Z=-0.0076,W=0.9924),Translation=(X=0,Y=0,Z=0),Scale3D=(X=1,Y=1,Z=1))),(Item=(Type=Bone,Name=\"dyn_hat_rear_02\"),Transform=(Rotation=(X=0,Y=0,Z=0,W=1),Translation=(X=0,Y=0,Z=0 ...(truncated)"
    WeightMinimum=0
    Weight=@n1.OutputParam
   )
   > n2 = &RigUnit_SphericalPoseReader(
    ActiveRegionScaleFactors=(PositiveWidth=1,NegativeWidth=1,PositiveHeight=1,NegativeHeight=1)
    ActiveRegionSize=0.1
    Debug=(bDrawDebug=true,bDraw2D=False,bDrawLocalAxes=False,DebugScale=25,DebugSegments=20,DebugThickness=0.25)
    DriverAxis=(X=1,Y=-0.5,Z=0)
    DriverItem=(Type=Bone,Name="head")
    FalloffRegionScaleFactors=(PositiveWidth=1,NegativeWidth=1,PositiveHeight=1,NegativeHeight=1)
    FalloffSize=0.3
    FlipHeightScaling=False
    FlipWidthScaling=False
    OptionalParentItem=(Type=Bone,Name="spine_05")
    RotationOffset=(X=-90,Y=0,Z=25)
   )
   > &RigUnit_ModifyTransforms(
    ItemToModify="((Item=(Type=Bone,Name=\"dyn_hat_rear_01\"),Transform=(Rotation=(X=0,Y=0,Z=0,W=1),Translation=(X=0,Y=0,Z=5),Scale3D=(X=0.9,Y=1,Z=1))),(Item=(Type=Bone,Name=\"dyn_hat_MSTR\"),Transform=(Rotation=(X=0,Y=0,Z=0.0436,W=0.999),Translation=(X=0.5,Y=-0.5,Z=0),Scale3D= ...(truncated)"
    WeightMinimum=0
    Weight=RigVMDispatchNode(False=0, True=@n2.OutputParam, Condition=RigVMVariableNode() {variable=Emote}.Value) {template="DISPATCH_RigVMDispatch_If(in Condition,in True,in False,out Result)"}.Result
   )
   n3 = &RigUnit_SphericalPoseReader(
    ActiveRegionScaleFactors=(PositiveWidth=1,NegativeWidth=1,PositiveHeight=1,NegativeHeight=1)
    ActiveRegionSize=0.05
    Debug=(bDrawDebug=True,bDraw2D=False,bDrawLocalAxes=False,DebugScale=25,DebugSegments=20,DebugThickness=0.25)
    DriverAxis=(X=-1,Y=0,Z=0)
    DriverItem=(Type=Bone,Name="clavicle_r")
    FalloffRegionScaleFactors=(PositiveWidth=1,NegativeWidth=1,PositiveHeight=1,NegativeHeight=1)
    FalloffSize=0.1
    FlipHeightScaling=False
    FlipWidthScaling=False
    OptionalParentItem=(Type=Bone,Name="head")
    RotationOffset=(X=-2.4142,Y=145.4321,Z=-36.5014)
    ExecutePin=@RigVMFunction_Sequence.C
   )
   > &RigUnit_ModifyTransforms(
    ItemToModify="((Item=(Type=Bone,Name=\"dyn_hat_rear_01\"),Transform=(Rotation=(X=0.1589,Y=0.0507,Z=-0.0082,W=-0.986),Translation=(X=-2.8152,Y=-0.0911,Z=4.3461),Scale3D=(X=1,Y=1,Z=1))),(Item=(Type=Bone,Name=\"dyn_hat_side_r_01\"),Transform=(Rotation=(X=-0.1684,Y=0.2291,Z=-0. ...(truncated)"
    Mode=OverrideLocal
    WeightMinimum=0
    Weight=@n3.OutputParam
   )
   n4 = &RigUnit_GetRelativeTransformForItem(Child=(Type=Bone,Name="spine_05"), Parent=(Type=Bone,Name="neck_01"), bChildInitial=False, bParentInitial=False)
   Multiply = &RigVMFunction_MathDoubleMul(
    B=&RigVMFunction_MathDoubleRemap(
     SourceMaximum=-32
     SourceMinimum=-20
     TargetMaximum=1
     TargetMinimum=0
     Value=&RigVMFunction_MathQuaternionToEuler(RotationOrder=ZYX, Value=@n4.RelativeTransform.Rotation).Result.Z
    ).Result
    A=@n2.OutputParam
   )
   RigVMRerouteNode(Value=@RigVMFunction_Sequence.B)
   &RigUnit_SphericalPoseReader(
    ActiveRegionScaleFactors=(PositiveWidth=1,NegativeWidth=1,PositiveHeight=1,NegativeHeight=1)
    ActiveRegionSize=0.1
    Debug=(bDrawDebug=false,bDraw2D=False,bDrawLocalAxes=False,DebugScale=25,DebugSegments=20,DebugThickness=0.25)
    DriverAxis=(X=1,Y=0,Z=0)
    DriverItem=(Type=Bone,Name="head")
    ExecutePin=()
    FalloffRegionScaleFactors=(PositiveWidth=1,NegativeWidth=1,PositiveHeight=1,NegativeHeight=1)
    FalloffSize=0.3
    FlipHeightScaling=False
    FlipWidthScaling=False
    OptionalParentItem=(Type=Bone,Name="spine_05")
    RotationOffset=(X=180,Y=0,Z=20)
   )
   RigVMVariableNode() {variable=Crouching}
  ]}
]
hierarchy [
 root {loc=<0 0 0>, rot=<0 0 0 1>}
 pelvis {parent=root, loc=<0 -0.7071 93.1191>, rot=<0.0127 0.707 -0.0127 -0.707>}
 spine_01 {parent=pelvis, loc=<6.7494 -0.6388 0>, rot=<0 0 0.034 -0.9994>}
 spine_02 {parent=spine_01, loc=<7.2518 -0.035 0>, rot=<0 0 0.0238 -0.9997>}
 spine_03 {parent=spine_02, loc=<6.0104 0 0>, rot=<0 0 0.0149 -0.9999>}
 spine_04 {parent=spine_03, loc=<8.1024 0 0>, rot=<0 0 -0.0132 -0.9999>}
 spine_05 {parent=spine_04, loc=<10.0146 0 0>, rot=<0 0 -0.0015 -1>}
 neck_01 {parent=spine_05, loc=<11.3138 0.1529 0>, rot=<0 0 -0.2003 -0.9797>}
 neck_02 {parent=neck_01, loc=<4.5627 0 0>, rot=<0 0 0.0504 -0.9987>}
 head {parent=neck_02, loc=<5.3914 0 0>, rot=<0 0 0.115 -0.9934>}
 dyn_hat_MSTR {parent=head, loc=<0.0127 0.0272 -0.1924>, rot=<0 0 0 1>}
 dyn_hat_rear_r_MSTR {parent=dyn_hat_MSTR, loc=<1.9418 -8.3805 0.5049>, rot=<0.2786 -0.6006 0.6652 -0.3451>}
 dyn_hat_rear_01 {parent=dyn_hat_rear_r_MSTR, loc=<0 0 0>, rot=<0 0 0 1>}
 dyn_hat_rear_02 {parent=dyn_hat_rear_01, loc=<7.0864 0 0>, rot=<0 0.1553 0 -0.9879>}
 dyn_hat_side_r_MSTR {parent=dyn_hat_MSTR, loc=<4.7491 -3.7905 6.8768>, rot=<0.5775 -0.3685 0.7263 0.0567>}
 dyn_hat_side_r_01 {parent=dyn_hat_side_r_MSTR, loc=<0 0 0>, rot=<0 0 0 1>}
 dyn_hat_side_r_02 {parent=dyn_hat_side_r_01, loc=<6.9046 0 0>, rot=<-0.0634 0.0705 -0.0429 -0.9946>}
 dyn_neckNoTwist {parent=spine_05, loc=<11.3138 0.1529 0>, rot=<0 0 -0.2003 -0.9797>}
 dyn_backCollar {parent=dyn_neckNoTwist, loc=<-0.3535 -4.1344 0.0006>, rot=<0 0 0.1608 -0.987>}
 dyn_collar_l_MSTR {parent=spine_05, loc=<12 6 -7.7>, rot=<0.1693 -0.877 -0.1337 -0.4292>}
 clavicle_r {parent=spine_05, loc=<4.3575 -0.1466 2.3223>, rot=<0.7289 0.0465 -0.6829 0.009>}
]
variables [
 Crouching {type=double, default=0}
 Emote {type=bool, default=False}
]
